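; Animated Jemma statue building actor that tilts back and forth or bobs through a timeline named HeadBob.
; On BeginPlay it becomes invulnerable, stores the scene component's world rotation as the start, and derives an end rotation with added roll.
; After a delay, StartAnimation plays HeadBob, lerping world rotation between the two on each update.
; EndAnimation and the construction script are empty.
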(event EventBeginPlay
  (Variables|BuildingActor|SetbIsInvulnerable true)
  (bind _output_get (Variables|Default|SetStartRotation (Transformation|GetWorldRotation (Variables|Default|GetScene))))
  (Variables|Default|SetEndRotation (Math|Rotator|MakeRotator (+ (.roll _output_get) 1e+01) (.pitch _output_get) (.yaw _output_get)))
  (Utilities|FlowControl|Delay 1e+01)
  (CallFunction|StartAnimation))

(event Custom|StartAnimation
  (bind (_direction _lerp) (Timeline|HeadBob PlayFromStart)
    (:Update
      (Transformation|SetWorldRotation (Variables|Default|GetScene) (Math|Rotator|Lerp(Rotator) (Variables|Default|GetStartRotation) (Variables|Default|GetEndRotation) _lerp true) false false))))

(event Custom|EndAnimation)

(fn ConstructionScript)
